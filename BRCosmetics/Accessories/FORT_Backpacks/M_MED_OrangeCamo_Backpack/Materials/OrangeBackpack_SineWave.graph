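# Material function for the OrangeCamo backpack that layers a glowing, animated sine wave over a blue grid onto incoming material attributes.
# A tiled grid-paper texture with BumpOffset depth blends two blues, and a panned, center-scaled sine texture blends in the bright SineColor.
# The result, multiplied by the Fresnel input, adds to emissive while base color is set to a constant.
# The Alpha input blends this against the original attributes.
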MaterialFunction /BRCosmetics/Accessories/FORT_Backpacks/M_MED_OrangeCamo_Backpack/Materials/OrangeBackpack_SineWave
graph {nodes [
  n1 = Multiply(A=TextureCoordinate(), B=ScalarParameter() {defaultValue=8, parameterName=GridTiling})
  n2 = ScalarParameter() {defaultValue=-0.2, parameterName=GridHeight}
  n3 = LinearInterpolate(
   A=VectorParameter() {parameterName=DeeperBlue, defaultValue=<rgba 0 0.0502 0.1146 1>}.RGB
   B=VectorParameter() {parameterName=DeepBlue, defaultValue=<rgba 0.0449 0.1453 0.2396 1>}.RGB
   Alpha=TextureSample(UVs=BumpOffset(Coordinate=@n1, Height=@n2)) {texture=/BRCosmetics/UI/Frontend/Upgrades/Textures/Pattern_GridPaper_10x}.A
  )
  n4 = VectorParameter() {parameterName=SineColor, defaultValue=<rgba 0 5.542 8 1>}
  n5 = Add(
   A=TextureCoordinate()
   B=AppendVector(A=ScalarParameter() {defaultValue=-0.1524, parameterName=SineOffset_X}, B=ScalarParameter() {defaultValue=0.349, parameterName=SineOffset_Y})
  )
  n6 = AppendVector(A=ScalarParameter() {defaultValue=0.2, parameterName=speed}, B=Constant())
  n7 = ScalarParameter() {defaultValue=0.15, parameterName=ScaleSine_X}
  n8 = ScalarParameter() {defaultValue=0.0619, parameterName=ScaleSine_Y}
  n9 = TextureSample(UVs=MaterialFunctionCall(UVs=Panner(Coordinate=@n5, Speed=@n6), Texture Scale=AppendVector(A=@n7, B=@n8)) {materialFunction=&ScaleUVsByCenter}.UVs) {texture=/BRCosmetics/Weapons/WeaponSkins/Textures/T_SimpleSIne2}
  Add(A=@n3, B=Multiply(A=@n4.RGB, B=@n9.R))
  n10 = FunctionInput() {inputType=FunctionInput_MaterialAttributes}
  n11 = GetMaterialAttributes(None=@n10) {attributeGetTypes [#1]}
  n12 = Add(
   A=@n11.EmissiveColor
   B=Multiply(A=LinearInterpolate(A=@n3, B=@n4.RGB, Alpha=@n9.R), B=FunctionInput() {inputName=Fresnel, inputType=FunctionInput_Scalar})
  )
  n13 = Constant()
  n14 = FunctionInput() {inputName=Alpha, inputType=FunctionInput_Scalar}
  FunctionOutput(None=BlendMaterialAttributes(
    A=@n10
    B=SetMaterialAttributes!(MaterialAttributes=@n11.MaterialAttributes, EmissiveColor=@n12, BaseColor=@n13) {attributeSetTypes [#1, #2]}
    Alpha=@n14
   ))
  Multiply(A=@n9.R, B=@n14)
 ]}
