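# Master material for the Speedy backpack cosmetic in the OrangeCamo backpack folder, using material attributes and Default Lit shading on skeletal meshes.
# The shared character PreFX function feeds an OrangeBackpack_SineWave effect function, which also takes the red channel of an FX mask texture parameter as alpha.
# The result passes through the shared PostFX function, keeping the backpack effect between the two standard character shading stages.
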Material /BRCosmetics/Accessories/FORT_Backpacks/M_MED_OrangeCamo_Backpack/Materials/M_Med_SpeedyBackpack
graph {
 properties {
  bUseMaterialAttributes=true
  bUsedWithMeshDeformer=true
  bUsedWithSkeletalMesh=true
  pixelDepthOffsetMode=PDOM_Legacy
  usedShadingModels=Default Lit
 }
 nodes [
  n1 = MaterialFunctionCall() {materialFunction=&BaseFN_CharacterFunction_PreFX}
  n2 = TextureSampleParameter2D() {
   parameterName=FX Mask
   samplerType=SAMPLERTYPE_Masks
   texture=/BRCosmetics/Accessories/FORT_Backpacks/M_MED_OrangeCamo_Backpack/Textures/T_M_MED_OrangeCamo_Backpack_FX
  }
  MaterialFunctionCall(
   AO=@n1.AO
   Distance Blend=@n1.AO
   Fresnel=@n1.AO
   Material=MaterialFunctionCall!(Fresnel=@n1.Fresnel, Alpha=@n2.R, In=@n1.Fresnel) {materialFunction=&/OrangeBackpack_SineWave}.Result
  ) {materialFunction=&BaseFN_CharacterFunction_PostFX}
 ]
}
