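# Physics asset for the orange camo backpack cosmetic accessory, giving its skeletal mesh simple collision and a jointed dynamic segment.
# One capsule body sits on backpack_root and a smaller one on the dyn_long_backpack bone.
# A single constraint joins the two, locking twist and the first swing axis so the second swing axis is the only rotation left unlocked.
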PhysicsAsset /BRCosmetics/Accessories/FORT_Backpacks/M_MED_OrangeCamo_Backpack/Meshes/M_MED_OrangeCamo_Backpack_PhysicsAsset
bodies {
 backpack_root [{shapeType=Capsule, length=15.6682, radius=7.8427, center=<1.2855 -8.1368 0.2586>, rotation=<rot 0 -87.03 -91.0809>}]
 dyn_long_backpack [{shapeType=Capsule, length=11.7132, radius=5.724, center=<3.9621 0.0414 0.1893>, rotation=<rot 0 -0.0624 -89.8316>}]
}
constraints [{
  bone1Name=dyn_long_backpack
  bone2Name=backpack_root
  swing1LimitDegrees=5
  swing1Motion=Locked
  swing2LimitDegrees=5
  twistLimitDegrees=45
  twistMotion=Locked
 }]
